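# M_DebugWaterPlane is an unlit debug material in FortWater, apparently for viewing a texture or normal map on a water plane.
# It samples a Texture parameter and a Normal parameter, keeping the RGB channels of each.
# The normal goes through ConstantBiasScale to remap it into a displayable color range.
# The DisplayNormal scalar blends between the texture color and that remapped normal.
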Material /FortWater/Materials/Debug/M_DebugWaterPlane
graph {
 properties {bUsedWithNanite=true, pixelDepthOffsetMode=PDOM_Legacy, shadingModel=MSM_Unlit, usedShadingModels=Unlit}
 nodes [
  n1 = TextureSampleParameter2D() {parameterName=Normal, samplerType=SAMPLERTYPE_Normal, texture=/Engine/EngineMaterials/DefaultNormal}
  LinearInterpolate(
   A=ComponentMask(None=TextureSampleParameter2D() {parameterName=Texture}.RGB) {b=true, g=true, r=true}
   B=ConstantBiasScale(None=ComponentMask(None=@n1.RGB) {b=true, g=true, r=true})
   Alpha=ScalarParameter() {parameterName=DisplayNormal}
  )
 ]
}
